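# Translucent, unlit, two-sided Niagara mesh particle material, named for the side of the sawblades in the OrderGuard emote effect.
# Particle color feeds an emissive adjustment function.
# The alpha chain multiplies six repeating, reflected-gradient bands along U by time-panned noise and an absolute normal-to-camera dot that fades grazing angles, then applies an exponent and particle alpha.
# A separate V-based softness branch also multiplies the bands but does not feed that chain.
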Material /BRCosmetics/Athena/Effects/Particles/Emotes/NS_Emote_OrderGuard/Materials/M_Emote_OrderGuard_Sawblades_Side
graph {
 properties {
  bUsedWithNiagaraMeshParticles=true
  blendMode=BLEND_Translucent
  pixelDepthOffsetMode=PDOM_Legacy
  shadingModel=MSM_Unlit
  twoSided=true
  usedShadingModels=Unlit
 }
 nodes [
  n1 = ParticleColor()
  MaterialFunctionCall(Emissive=@n1.RGB) {materialFunction=&MF_FX_EmissiveAdjustments}
  n2 = TextureCoordinate()
  n3 = ComponentMask(None=@n2) {g=true}
  n4 = ScalarParameter() {parameterName=Top Softness}
  n5 = Add(A=ComponentMask(None=@n2) {r=true}, B=ScalarParameter() {defaultValue=0.08, parameterName=Offset})
  n6 = MaterialFunctionCall!(Gradient=Frac(None=Multiply(A=@n5) {constB=6}), Center Point=ScalarParameter() {defaultValue=0.2, parameterName=Center Point}) {materialFunction=/Game/Effects/Fort_Effects/MaterialFunctions/MF_GradientReflected}
  Multiply(A=Saturate(None=Divide(A=@n3, B=@n4)), B=@n6.Result)
  n7 = TextureCoordinate()
  n8 = AppendVector(A=Time(), B=Constant())
  n9 = Multiply(
   A=@n6.Result
   B=TextureSample(UVs=Add(A=@n7, B=@n8)) {samplerType=SAMPLERTYPE_LinearColor, texture=/Game/Global/Textures/Default/Noise/T_FN_FabricTile_B}.R
  )
  n10 = Abs(None=DotProduct(A=VertexNormalWS(), B=CameraVectorWS()))
  Multiply(A=@n1.A, B=Power(Base=Multiply(A=@n9, B=@n10), Exp=ScalarParameter() {defaultValue=2, parameterName=Exponent}))
 ]
}
